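# Idle body Niagara effect for the BoldDormRise character cosmetic, adding shoulder flames, belt flames and drifting sparks on skeletal mesh bones.
# Each flame emitter bursts two persistent, bone-following particles, using partition switches for per-side placement, feeding mesh, glow and fake-light renderers.
# Sparks spawn continuously across four partitions with curl noise.
# User.Variant blends orange to white; ADS, equip fade-in, front-end preview and time-of-day adjust flames; BeltFlames001 is disabled.
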NiagaraSystem /BRCosmetics/Characters/Player/Male/Medium/Bodies/M_MED_BoldDormRise_Frock/FX/BodyFX/NS_Character_BoldDormRise_Idle_Body
dependencies {
 usedDataInterfaces [
  &NiagaraDataInterfaceSpriteRendererInfo
  &NiagaraDataInterfaceCurve
  &NiagaraDataInterfaceColorCurve
  &NiagaraDataInterfaceSkeletalMesh
 ]
 usedDynamicInputs [&ClampFloat, &SetBoolByFloatComparison]
 usedModules [
  &self:SystemSpawnScript.SystemScriptSource.SystemScriptGraph.NiagaraNodeAssignment_0.SetVariables_#1
  &SystemState
  &self:SystemSpawnScript.SystemScriptSource.SystemScriptGraph.NiagaraNodeAssignment_4.SetVariables_#2
  &Timeline
  &self:SystemSpawnScript.SystemScriptSource.SystemScriptGraph.NiagaraNodeAssignment_2.SetVariables_#3
  &FN_VariantManager
  &self:SystemSpawnScript.SystemScriptSource.SystemScriptGraph.NiagaraNodeAssignment_1.SetVariables_#4
  &self:SystemSpawnScript.SystemScriptSource.SystemScriptGraph.NiagaraNodeAssignment_3.SetVariables_#5
 ]
}
emitters {
 BasicSparks {
  emitterUpdateScript [
   &EmitterState()
   &SpawnRate(SpawnRate=16, Use Spawn Probability=false)
  ]
  particleSpawnScript [
   &PartitionParticles(Number Of Partitions=4)
   &InitializeParticle(
    Color=&LerpLinearColors(
     StartColor=&RandomRangeLinearColor(Minimum=<rgba 1 0.6933 0.2 1>, Maximum=<rgba 1 0.7625 0.05 1>)
     EndColor=&RandomRangeLinearColor(Minimum=<rgba 1 1 1 1>, Maximum=<rgba 0.7 0.7 0.7 1>)
     LerpFactor=$User.Variant
    )
    Lifetime=&RandomRangeFloat(Minimum=0.4, Maximum=0.7)
    Sprite Size=&Multiply_Vector2D_ByFloat(
     Float=&RandomRangeFloat(Minimum=0.4, Maximum=0.6)
     Vector2D=&RandomRangeVector2D(Minimum=<1 2>, Maximum=<1.3 4>)
    )
    Sprite Size Mode=::NewEnumerator4
    UsePositionOffset=false
    Write Material Random=true
   )
   &SkeletalMeshLocation(
    Bone / Socket Index=&MergeHelper_IntSwitch(0=0, 1=1, 2=2, 3=2, 4=?, 5=?, Index=$Particles.PartitionParticles.Partition, Num=::NewEnumerator2)
    Bone Sampling Mode=::NewEnumerator4
    MaskSpawnGroup=false
    Sampled Position Offset=&MergeHelper_VectorSwitch(0=<13 0 3>, 1=<-13 0 -3>, 2=<5 10 0>, 3=<5 -12 0>, Index=$Particles.PartitionParticles.Partition, Num=::NewEnumerator2)
    Skeletal Mesh=$System.Skeletal Mesh
   )
   &SubUVAnimation(
    "Sprite Renderer "=DataInterface() {propertyValues {SpriteRenderer=&self:BasicSparks_0.NiagaraSpriteRendererProperties_0}}
    SubUV Animation Mode:ENiagara_SubUVLookupModeV2=::NewEnumerator2
    UseEndFrame=false
    UseStartFrame=false
   ) {enabled=false}
   &ShapeLocation(
    MaskSpawnGroup=false
    Non Uniform Scale=&MergeHelper_VectorSwitch(0=<1.5 1.5 1.5>, 1=<1.5 1.5 1.5>, 2=<3 1 1>, 3=<3 1 1>, Index=$Particles.PartitionParticles.Partition, Num=::NewEnumerator2)
    Override Seed=false
    Sphere Radius=5
    Sphere Surface Distribution=0.2143
    UseOwnerScale=true
   )
   &AddVelocity(
    Module.Use Falloff Distance=false
    Override Seed=false
    Velocity Mode=::NewEnumerator1
    Velocity Origin=$Particles.SkeletalMeshLocation.SampledPosition
    Velocity Speed=&Multiply_Float(
     A=&RandomRangeFloat(Minimum=20, Maximum=40)
     B=&MergeHelper_FloatSwitch(
      0=1
      1=0.6
      2=?
      3=?
      Index=/Niagara/DynamicInputs/Divide/Divide_Int(A:NiagaraInt32=$Particles.PartitionParticles.Partition, B:NiagaraInt32=2)
      Num=::NewEnumerator0
     )
    )
   )
   &InheritVelocity(Inherited Velocity Amount Scale=&VectorFromFloat(Value=0.4), Speed Limit=true, UseSpeedThreshold=false) {enabled=false}
   &CameraOffset(Camera Offset Amount=4)
  ]
  particleUpdateScript [
   &ParticleState()
   &ScaleSpriteSize(Uniform Curve Sprite Scale=DataInterface() {propertyValues {Curve {keys [
        {interpMode=RCIM_Cubic, tangentMode=RCTM_User, time=0, value=0}
        {interpMode=RCIM_Cubic, tangentMode=RCTM_User, time=0.2, value=1}
        {interpMode=RCIM_Cubic, tangentMode=RCTM_User, time=1, value=0}
       ]}}})
   &InheritSourceMovement(
    Applied Position Scale=&VectorFromFloat(Value=&FloatFromCurve(
      FloatCurve=DataInterface() {propertyValues {Curve {keys [
          {interpMode=RCIM_Cubic, tangentMode=RCTM_User, time=0.5, value=1}
          {arriveTangent=-2.6042, interpMode=RCIM_Cubic, leaveTangent=-2.6042, tangentMode=RCTM_User, time=1, value=0}
         ]}}}
      Scale Curve=&RandomRangeFloat(Minimum=0.9, Maximum=1)
     ))
    ApplyForce=false
    ApplyPosition=true
    ApplyPositionFalloff=false
   )
   &CurlNoiseForce(
    Bias Noise=false
    Noise Frequency=&RandomRangeFloat(Minimum=60, Maximum=90)
    Noise Strength=&Multiply_Float(
     A=&FloatFromCurve(
      FloatCurve=DataInterface() {propertyValues {Curve {keys [{arriveTangent=2.3913, interpMode=RCIM_Cubic, leaveTangent=2.3913, tangentMode=RCTM_User, time=0, value=0.4}, <at 0.3 1>]}}}
      Scale Curve=&RandomRangeFloat(Minimum=120, Maximum=200)
     )
     B=&MergeHelper_FloatSwitch(
      0=1
      1=0.6
      2=?
      3=?
      Index=&Divide_Int(A:NiagaraInt32=$Particles.PartitionParticles.Partition, B:NiagaraInt32=2)
      Num=::NewEnumerator0
     )
    )
    Pan Noise=true
    Pan Noise Field=<0.1 0 0>
   )
   &Drag(Drag=&RandomRangeFloat(Minimum=2, Maximum=5), Use Linear Drag=true, Use Rotational Drag=true)
   &SolveForcesAndVelocity(Clamp Velocity=false, Limit Acceleration=false)
   &ScaleSpriteSizeBySpeed(Max Scale Factor=<0.8 1>, Min Scale Factor=<1 0.5>, Velocity Threshold=40)
   &DynamicMaterialParameters(
    Index 0 Param 1=0.1
    Index 0 Param 2=4
    Module.Index0Float0Write=true
    Module.Index0Float1Write=true
    Module.Index0Float2Write=false
    Module.Index0Float3Write=false
   )
   &ScaleColor(
    Scale Alpha=&Lerp_Float(A=1, B=0.025, Alpha=$System.AlphaADS)
    Scale RGB=&MakeVectorFromLinearColorRGB(Color=&LerpLinearColors(
      StartColor=&ColorFromCurve(ColorCurve=DataInterface() {propertyValues {
         AlphaCurve {keys [<at 0 1 RCIM_Cubic>]}
         BlueCurve {keys [<at 0.0943 7 RCIM_Cubic>,<at 0.3611 2.0955>,<at 1 0.25>]}
         GreenCurve {keys [<at 0.0943 8.5 RCIM_Cubic>,<at 0.3611 2.547>,<at 1 0.3583>]}
         RedCurve {keys [<at 0.0943 10 RCIM_Cubic>,<at 0.3611 3>,<at 1 0.5>]}
        }})
      EndColor=&ColorFromCurve(ColorCurve=DataInterface() {propertyValues {
         AlphaCurve {keys [<at 0 1 RCIM_Cubic>]}
         BlueCurve {keys [<at 0.0943 10 RCIM_Cubic>,<at 0.3722 3>,<at 1 0.2>]}
         GreenCurve {keys [<at 0.0943 10 RCIM_Cubic>,<at 0.3722 3>,<at 1 0.2>]}
         RedCurve {keys [<at 0.0943 10 RCIM_Cubic>,<at 0.3722 3>,<at 1 0.2>]}
        }})
      LerpFactor=$User.Variant
     ))
    ScaleA=true
    ScaleRGB=true
   )
   /SetVariables(
    Particles.ColorGlow:LinearColor=&MultiplyLinearColor(
     Linear Color=&LerpLinearColors(StartColor=<rgba 0.15 0.125 0.075 0.2>, EndColor=<rgba 0.1 0.1 0.1 0.2>, LerpFactor=$User.Variant)
     Scale Factor=&MakeVector4FromVectorAndFloat(
      "Vector (RGB)"=&MakeVectorFromLinearColorRGB(Color=$Particles.Color)
      "Float (Alpha)"=&Lerp_Float(A=1, B=1, Alpha=$System.AlphaADS)
     )
    )
    Particles.DynParGlow:Vector4f=<1 1 1 0>
    Particles.SpriteSizeGlow:Vector2f=&Multiply_Vector2D(A=$Particles.SpriteSize, B=<8 4>)
   ) {isSetParameters=true}
  ]
  rendererClasses [&NiagaraSpriteRendererProperties]
  renderers [
   {rendererClass=&NiagaraSpriteRendererProperties}
   {rendererClass=&NiagaraSpriteRendererProperties, rendererIndex=1}
  ]
 }
 BeltFlames {
  emitterUpdateScript [
   &EmitterState(Life Cycle Mode=::NewEnumerator1, Loop Behavior=::NewEnumerator1, Loop Duration Mode=::NewEnumerator1, UseLoopDelay=false)
   &SpawnBurst_Instantaneous(Spawn Count=2, Use Spawn Probability=false, UseLoopCountLimit=true)
  ]
  particleSpawnScript [
   &PartitionParticles(ID to Partition:NiagaraInt32=&ReturnExecIndex())
   &InitializeParticle(
    Mesh Scale=&MergeHelper_VectorSwitch(0=<1.2 1 1.5>, 1=<1 -1 1.1>, 2=?, 3=?, Index=$Particles.PartitionParticles.Partition, Num=::NewEnumerator0)
    Mesh Scale Mode=::NewEnumerator4
    Sprite Size Mode=::NewEnumerator3
    Uniform Sprite Size=16
    UsePositionOffset=false
    Write Material Random=true
   )
   &CameraOffset(Camera Offset Amount=&MergeHelper_FloatSwitch(0=2, 1=2.2, 2=?, 3=?, Index=$Particles.PartitionParticles.Partition, Num=::NewEnumerator0))
   &InitialMeshOrientation(
    Mesh Axis To Orient=&MergeHelper_VectorSwitch(0=<0 0 1>, 1=<0 0 -1>, 2=?, 3=?, Index=$Particles.PartitionParticles.Partition, Num=::NewEnumerator0)
    Mesh Orientation Mode=::NewEnumerator2
    Rotation=&MergeHelper_VectorSwitch(0=<0 0 0>, 1=<0.5 0 0>, 2=?, 3=?, Index=$Particles.PartitionParticles.Partition, Num=::NewEnumerator0)
    Use Rotation Vector=true
   )
   /SetVariables(Particles.ColorFlame:LinearColor=&LerpLinearColors(StartColor=<rgba 3 2 1.5 1>, EndColor=<rgba 1 1 1 1>, LerpFactor=$User.Variant)) {enabled=false, isSetParameters=true, module=SetVariables_#6}
   /SetVariables(Particles.ScaleFlameUpper:Vector3f=&Multiply_VectorByFloat(
     Vector=<1 1 1>
     Float=&MergeHelper_FloatSwitch(0=1, 1=-1, 2=?, 3=?, Index=$Particles.PartitionParticles.Partition, Num=::NewEnumerator0)
    )) {isSetParameters=true, module=SetVariables_#7}
  ]
  particleUpdateScript [
   &SkeletalMeshLocation(
    Bone / Socket Index=2
    Bone Sampling Mode=::NewEnumerator4
    MaskSpawnGroup=false
    Sampled Position Offset=&MergeHelper_VectorSwitch(0=<7 8.5 0>, 1=<4 -14 0>, 2=?, 3=?, Index=$Particles.PartitionParticles.Partition, Num=::NewEnumerator0)
    Skeletal Mesh=$System.Skeletal Mesh
   )
   &ParticleState(Kill Particles When Lifetime Has Elapsed=false, Let Infinitely Lived Particles Die When Emitter Deactivates=true)
   &ScaleSpriteSize(Scale Sprite Size Mode=::NewEnumerator0, Uniform Scale Factor=&Lerp_Float(A=1, B=1.25, Alpha=$System.PlumeLength))
   &ScaleColor(
    Scale Alpha=&Multiply_Float(A=$System.EquipFadeIn, B=&Lerp_Float(A=1, B=0.33, Alpha=$System.AlphaADS))
    Scale RGB=&MakeVectorFromLinearColorRGB(Color=&LerpLinearColors(StartColor=<rgba 25 7.5 0 1>, EndColor=<rgba 5 5 5 1>, LerpFactor=$User.Variant))
    ScaleA=true
    ScaleRGB=true
   )
   /SetVariables(
    Particles.ColorFakeLight:LinearColor=&MultiplyLinearColor(
     Linear Color=&LerpLinearColors(
      StartColor=<rgba 5 1.8333 0.25 1>
      EndColor=&MakeCustomLinearColorFromBool(Bool=$User.bIsFrontEndPreview, True Color=<rgba 11 11 11 0.7>, False Color=<rgba 11 11 11 1>)
      LerpFactor=$User.Variant
     )
     Scale Factor=&MakeVector4FromVectorAndFloat("Vector (RGB)"=<1 1 1>, "Float (Alpha)"=$System.EquipFadeIn)
    )
    Particles.PositionFakeLight:NiagaraPosition=&AddVectorToPosition(
     Position=$Particles.Position
     Vector=&RotateVectorByQuaternion(
      Vector To Rotate=&MergeHelper_VectorSwitch(0=<-5 2 0>, 1=<-4 0 0>, 2=?, 3=?, Index=$Particles.PartitionParticles.Partition, Num=::NewEnumerator0)
      Quaternion=$Particles.SkeletalMeshLocation.SampledRotation
     )
    )
    Particles.SpriteSizeFakeLight:Vector2f=&Multiply_Vector2D_ByFloat(Float=50, Vector2D=<1 1>)
   ) {isSetParameters=true, module=SetVariables_#8}
   /SetVariables(
    Particles.ColorGlow:LinearColor=&Multiply_LinearColorByFloat(
     Linear Color=&LerpLinearColors(
      StartColor=<rgba 1 0.4575 0.07 0.25>
      EndColor=&MakeCustomLinearColorFromBool(
       Bool=$User.bIsFrontEndPreview
       True Color=<rgba 1 1 1 0.05>
       False Color=&FNTimeOfDayColor(Night Color=<rgba 1 1 1 0.1>, Morning Color=<rgba 1 1 1 0.1>, Day Color=<rgba 1 1 1 0.1>, Evening Color=<rgba 1 1 1 0.1>)
      )
      LerpFactor=$User.Variant
     )
     Float=&Multiply_Float(A=$System.EquipFadeIn, B=&Lerp_Float(A=1, B=0.5, Alpha=$System.AlphaADS))
    )
    Particles.PositionGlow:NiagaraPosition=&AddVectorToPosition(
     Position=$Particles.Position
     Vector=&RotateVectorByQuaternion(
      Vector To Rotate=&MergeHelper_VectorSwitch(0=<-4 1 0>, 1=<-3 4 0>, 2=?, 3=?, Index=$Particles.PartitionParticles.Partition, Num=::NewEnumerator0)
      Quaternion=$Particles.SkeletalMeshLocation.SampledRotation
     )
    )
   ) {isSetParameters=true, module=SetVariables_#9}
   /SetVariables(
    Particles.CameraOffsetGlow:NiagaraFloat=&MergeHelper_FloatSwitch(0=0, 1=5, 2=?, 3=?, Index=$Particles.PartitionParticles.Partition, Num=::NewEnumerator0)
    Particles.SpriteSizeGlow:Vector2f=&Multiply_Vector2D_ByFloat(Float=3, Vector2D=$Particles.SpriteSize)
   ) {isSetParameters=true, module=SetVariables_#10}
   &FN_Light_Attributes(
    Color=&LerpLinearColors(StartColor=<rgba 36 16.47 2.52 1>, EndColor=<rgba 10 10 10 1>, LerpFactor=$User.Variant)
    Create Color Attribute=true
    Scale Color=&Waveform(
     Waveform Count=::NewEnumerator1
     "Waveform [1]"=::NewEnumerator0
     "[1] Input Position"=0.1
     "[1] Frequency"=0.433
     "[1] Amplitude Min/Max"=<0.4 1>
     "Waveform [2] Blend"=::NewEnumerator4
     "[2] Input Position"=0
     "[2] Phase"=$Emitter.Age
     "[2] Phase Scale"=1
     "[2] Compound Frequency"=2
     "[2] Frequency"=3.1
     "[2] Amplitude Min/Max"=<0.8 1>
     "[2] Scale"=1
     "[2] Bias"=0
     "[2] Exponent"=1
     "[2] Pulse Duration"=0.1
     "[2] Clip Triangle Min/Max"=<0 0>
     "[2] Random Seed"=0
    )
   ) {enabled=false}
   &DynamicMaterialParameters(
    Index 0 Param 2=10
    Index 0 Param 3=0.5
    Module.Index0Float0Write=false
    Module.Index0Float1Write=true
    Module.Index0Float2Write=true
    Module.Index0Float3Write=false
   )
   /SetVariables(
    Particles.ColorFlame:LinearColor=&MultiplyLinearColor(
     Linear Color=&LerpLinearColors(StartColor=<rgba 3 2 1.5 1>, EndColor=<rgba 0.66 0.66 0.66 3>, LerpFactor=$User.Variant)
     Scale Factor=&MakeVector4FromVectorAndFloat(
      "Vector (RGB)"=&VectorFromFloat(Value=&Lerp_Float(A=1, B=1.15, Alpha=$User.Variant))
      "Float (Alpha)"=&Multiply_Float(
       A=$System.EquipFadeIn
       B=&Multiply_Float(
        A=&Lerp_Float(A=1, B=1, Alpha=$System.AlphaADS)
        B=&Lerp_Float(A=0.01, B=0.2, Alpha=$User.Variant)
       )
      )
     )
    )
    Particles.PositionFlame:NiagaraPosition=&AddVectorToPosition(
     Position=$Particles.Position
     Vector=&RotateVectorByQuaternion(
      Vector To Rotate=&MergeHelper_VectorSwitch(0=<0 0 0>, 1=<0 0 0>, 2=?, 3=?, Index=$Particles.PartitionParticles.Partition, Num=::NewEnumerator0)
      Quaternion=$Particles.SkeletalMeshLocation.SampledRotation
     )
    )
   ) {isSetParameters=true, module=SetVariables_#11}
   /SetVariables(Particles.OrientationUpperFlames:Quat4f=&MultiplyQuaternion(
     Quaternion A=$Particles.MeshOrientation
     Quaternion B=/Game/Effects/Niagara/Modules/MergeEmitters/MergeHelper_QuatSwitch(
      0:Quat4f=<-0.1 0.15 0 1>
      1:Quat4f=<-0.02 0.15 0 1>
      2:Quat4f=?
      3:Quat4f=?
      4:Quat4f=?
      5:Quat4f=?
      6:Quat4f=?
      7:Quat4f=?
      8:Quat4f=?
      9:Quat4f=?
      Index:NiagaraInt32=$Particles.PartitionParticles.Partition
      Num:MergeHelper_SwitcherEnum=::NewEnumerator0
     )
    )) {enabled=false, isSetParameters=true, module=SetVariables_#12}
  ]
  rendererClasses [&NiagaraSpriteRendererProperties, &NiagaraMeshRendererProperties]
  renderers [
   {rendererClass=&NiagaraSpriteRendererProperties}
   {rendererClass=&NiagaraSpriteRendererProperties, rendererIndex=1}
   {rendererClass=&NiagaraMeshRendererProperties, rendererIndex=2}
   {rendererClass=&NiagaraMeshRendererProperties, rendererIndex=3}
   {rendererClass=&NiagaraMeshRendererProperties, rendererIndex=4}
  ]
 }
 BeltFlames001 {
  bEnabled=false
  emitterUpdateScript [
   &EmitterState(Life Cycle Mode=::NewEnumerator1, Loop Behavior=::NewEnumerator1, Loop Duration Mode=::NewEnumerator1, UseLoopDelay=false)
   &SpawnBurst_Instantaneous(Spawn Count=2, Use Spawn Probability=false, UseLoopCountLimit=true)
  ]
  particleSpawnScript [
   &PartitionParticles(ID to Partition:NiagaraInt32=&ReturnExecIndex())
   &InitializeParticle(
    Mesh Scale=&MergeHelper_VectorSwitch(0=<1.2 1 1.5>, 1=<1 -1 0.9>, 2=?, 3=?, Index=$Particles.PartitionParticles.Partition, Num=::NewEnumerator0)
    Mesh Scale Mode=::NewEnumerator4
    Sprite Size Mode=::NewEnumerator3
    Uniform Sprite Size=16
    UsePositionOffset=false
    Write Material Random=true
   )
   &CameraOffset(Camera Offset Amount=&MergeHelper_FloatSwitch(0=1, 1=2, 2=?, 3=?, Index=$Particles.PartitionParticles.Partition, Num=::NewEnumerator0))
   &InitialMeshOrientation(
    Mesh Axis To Orient=<0 0 1>
    Mesh Orientation Mode=::NewEnumerator2
    Rotation=&MergeHelper_VectorSwitch(0=<0 0 0>, 1=<-0.3 0 0>, 2=?, 3=?, Index=$Particles.PartitionParticles.Partition, Num=::NewEnumerator0)
    Use Rotation Vector=true
   )
   /SetVariables(Particles.ColorFlame:LinearColor=&LerpLinearColors(StartColor=<rgba 3 2 1.5 1>, EndColor=<rgba 1 1 1 1>, LerpFactor=$User.Variant)) {isSetParameters=true, module=SetVariables_#6}
   /SetVariables(Particles.ScaleFlameUpper:Vector3f=&Multiply_VectorByFloat(
     Vector=<1 1 1>
     Float=&MergeHelper_FloatSwitch(0=1, 1=-1, 2=?, 3=?, Index=$Particles.PartitionParticles.Partition, Num=::NewEnumerator0)
    )) {isSetParameters=true, module=SetVariables_#7}
  ]
  particleUpdateScript [
   &SkeletalMeshLocation(
    Bone / Socket Index=&Add_Integer(A=$Particles.PartitionParticles.Partition, B=2)
    Bone Sampling Mode=::NewEnumerator4
    MaskSpawnGroup=false
    Sampled Position Offset=&MergeHelper_VectorSwitch(0=<7 8.5 0>, 1=<-0.33 0 -1>, 2=?, 3=?, Index=$Particles.PartitionParticles.Partition, Num=::NewEnumerator0)
    Skeletal Mesh=$System.Skeletal Mesh
   )
   &ParticleState(Kill Particles When Lifetime Has Elapsed=false, Let Infinitely Lived Particles Die When Emitter Deactivates=true)
   &ScaleSpriteSize(Scale Sprite Size Mode=::NewEnumerator0, Uniform Scale Factor=&Lerp_Float(A=1, B=1.25, Alpha=$System.PlumeLength))
   &ScaleColor(
    Scale Alpha=&Lerp_Float(A=1, B=0.33, Alpha=$System.AlphaADS)
    Scale RGB=&MakeVectorFromLinearColorRGB(Color=&LerpLinearColors(StartColor=<rgba 25 7.5 0 1>, EndColor=<rgba 5 5 5 1>, LerpFactor=$User.Variant))
    ScaleA=true
    ScaleRGB=true
   )
   /SetVariables(
    Particles.ColorFakeLight:LinearColor=&LerpLinearColors(StartColor=<rgba 8 2.9333 0.4 1>, EndColor=<rgba 10 10 10 1>, LerpFactor=$User.Variant)
    Particles.PositionFakeLight:NiagaraPosition=&AddVectorToPosition(
     Position=$Particles.Position
     Vector=&RotateVectorByQuaternion(
      Vector To Rotate=&MergeHelper_VectorSwitch(0=<5 0 0>, 1=<2 0 0>, 2=?, 3=?, Index=$Particles.PartitionParticles.Partition, Num=::NewEnumerator0)
      Quaternion=$Particles.SkeletalMeshLocation.SampledRotation
     )
    )
    Particles.SpriteSizeFakeLight:Vector2f=&Multiply_Vector2D_ByFloat(Float=40, Vector2D=<1 1>)
   ) {isSetParameters=true, module=SetVariables_#8}
   /SetVariables(
    Particles.ColorGlow:LinearColor=&Multiply_LinearColorByFloat(
     Linear Color=&LerpLinearColors(StartColor=<rgba 1 0.4575 0.07 0.25>, EndColor=<rgba 1 1 1 0.3>, LerpFactor=$User.Variant)
     Float=&Lerp_Float(A=1, B=0.33, Alpha=$System.AlphaADS)
    )
    Particles.PositionGlow:NiagaraPosition=&AddVectorToPosition(
     Position=$Particles.Position
     Vector=&RotateVectorByQuaternion(
      Vector To Rotate=&MergeHelper_VectorSwitch(0=<12 0 3>, 1=<-12 0 -3>, 2=?, 3=?, Index=$Particles.PartitionParticles.Partition, Num=::NewEnumerator0)
      Quaternion=$Particles.SkeletalMeshLocation.SampledRotation
     )
    )
   ) {isSetParameters=true, module=SetVariables_#9}
   /SetVariables(
    Particles.CameraOffsetGlow=5
    Particles.SpriteSizeGlow:Vector2f=&Multiply_Vector2D_ByFloat(Float=3, Vector2D=$Particles.SpriteSize)
   ) {isSetParameters=true, module=SetVariables_#10}
   &FN_Light_Attributes(
    Color=&LerpLinearColors(StartColor=<rgba 36 16.47 2.52 1>, EndColor=<rgba 10 10 10 1>, LerpFactor=$User.Variant)
    Create Color Attribute=true
    Scale Color=&Waveform(
     Waveform Count=::NewEnumerator1
     "Waveform [1]"=::NewEnumerator0
     "[1] Input Position"=0.1
     "[1] Frequency"=0.433
     "[1] Amplitude Min/Max"=<0.4 1>
     "Waveform [2] Blend"=::NewEnumerator4
     "[2] Input Position"=0
     "[2] Phase"=$Emitter.Age
     "[2] Phase Scale"=1
     "[2] Compound Frequency"=2
     "[2] Frequency"=3.1
     "[2] Amplitude Min/Max"=<0.8 1>
     "[2] Scale"=1
     "[2] Bias"=0
     "[2] Exponent"=1
     "[2] Pulse Duration"=0.1
     "[2] Clip Triangle Min/Max"=<0 0>
     "[2] Random Seed"=0
    )
   ) {enabled=false}
   &DynamicMaterialParameters(
    Index 0 Param 3=0.5
    Module.Index0Float0Write=false
    Module.Index0Float1Write=false
    Module.Index0Float2Write=true
    Module.Index0Float3Write=false
   )
   /SetVariables(
    Particles.ColorFlame:LinearColor=&MultiplyLinearColor(
     Linear Color=$Particles.Initial.ColorFlame
     Scale Factor=&MakeVector4FromVectorAndFloat(
      "Vector (RGB)"=&VectorFromFloat(Value=&Lerp_Float(A=1, B=1.15, Alpha=$User.Variant))
      "Float (Alpha)"=&Multiply_Float(
       A=&Lerp_Float(A=1, B=0.33, Alpha=$System.AlphaADS)
       B=&Lerp_Float(A=0.026, B=0.4, Alpha=$User.Variant)
      )
     )
    )
    Particles.PositionFlame:NiagaraPosition=&AddVectorToPosition(
     Position=$Particles.Position
     Vector=&RotateVectorByQuaternion(
      Vector To Rotate=&MergeHelper_VectorSwitch(0=<0 0 0>, 1=<0 0 0>, 2=?, 3=?, Index=$Particles.PartitionParticles.Partition, Num=::NewEnumerator0)
      Quaternion=$Particles.SkeletalMeshLocation.SampledRotation
     )
    )
   ) {isSetParameters=true, module=SetVariables_#11}
   /SetVariables(Particles.OrientationUpperFlames:Quat4f=&MultiplyQuaternion(
     Quaternion A=$Particles.MeshOrientation
     Quaternion B=&MergeHelper_QuatSwitch(
      0:Quat4f=<-0.1 0.15 0 1>
      1:Quat4f=<-0.02 0.15 0 1>
      2:Quat4f=?
      3:Quat4f=?
      4:Quat4f=?
      5:Quat4f=?
      6:Quat4f=?
      7:Quat4f=?
      8:Quat4f=?
      9:Quat4f=?
      Index:NiagaraInt32=$Particles.PartitionParticles.Partition
      Num:MergeHelper_SwitcherEnum=::NewEnumerator0
     )
    )) {enabled=false, isSetParameters=true, module=SetVariables_#12}
  ]
  rendererClasses [&NiagaraSpriteRendererProperties, &NiagaraMeshRendererProperties]
  renderers [
   {rendererClass=&NiagaraSpriteRendererProperties}
   {rendererClass=&NiagaraSpriteRendererProperties, rendererIndex=1}
   {rendererClass=&NiagaraMeshRendererProperties, rendererIndex=2}
   {rendererClass=&NiagaraMeshRendererProperties, rendererIndex=3}
   {rendererClass=&NiagaraMeshRendererProperties, rendererIndex=4}
  ]
 }
 ShoulderFlames {
  emitterUpdateScript [
   &EmitterState(Life Cycle Mode=::NewEnumerator1, Loop Behavior=::NewEnumerator1, Loop Duration Mode=::NewEnumerator1, UseLoopDelay=false)
   &SpawnBurst_Instantaneous(Spawn Count=2, Use Spawn Probability=false, UseLoopCountLimit=true)
  ]
  particleSpawnScript [
   &PartitionParticles(ID to Partition:NiagaraInt32=&ReturnExecIndex())
   &InitializeParticle(
    Mesh Scale=&Multiply_VectorByFloat(
     Vector=<1.1 1 1.1>
     Float=&MergeHelper_FloatSwitch(0=1.1, 1=-1.1, 2=?, 3=?, Index=$Particles.PartitionParticles.Partition, Num=::NewEnumerator0)
    )
    Mesh Scale Mode=::NewEnumerator4
    Sprite Size Mode=::NewEnumerator3
    Uniform Sprite Size=16
    UsePositionOffset=false
    Write Material Random=true
   )
   &CameraOffset(Camera Offset Amount=1)
   &InitialMeshOrientation(
    Mesh Axis To Orient=&MergeHelper_VectorSwitch(0=<1 0 0>, 1=<1 0 0>, 2=?, 3=?, Index=$Particles.PartitionParticles.Partition, Num=::NewEnumerator0)
    Mesh Orientation Mode=::NewEnumerator2
    Rotation=&MergeHelper_VectorSwitch(0=<0 -0.11 0>, 1=<0 -0.11 0>, 2=?, 3=?, Index=$Particles.PartitionParticles.Partition, Num=::NewEnumerator0)
    Use Rotation Vector=true
   )
   /SetVariables(Particles.ColorFlame:LinearColor=&LerpLinearColors(StartColor=<rgba 3 2 1.5 1>, EndColor=<rgba 1 1 1 1>, LerpFactor=$User.Variant)) {enabled=false, isSetParameters=true, module=SetVariables_#6}
   /SetVariables(Particles.ScaleFlameUpper:Vector3f=&Multiply_VectorByFloat(
     Vector=<1 1 1>
     Float=&MergeHelper_FloatSwitch(0=1, 1=-1, 2=?, 3=?, Index=$Particles.PartitionParticles.Partition, Num=::NewEnumerator0)
    )) {isSetParameters=true, module=SetVariables_#7}
  ]
  particleUpdateScript [
   &SkeletalMeshLocation(
    Bone / Socket Index=$Particles.PartitionParticles.Partition
    Bone Sampling Mode=::NewEnumerator4
    MaskSpawnGroup=false
    Sampled Position Offset=&MergeHelper_VectorSwitch(0=<-3.5 -0.5 -1.2>, 1=<3.5 0.5 1.2>, 2=?, 3=?, Index=$Particles.PartitionParticles.Partition, Num=::NewEnumerator0)
    Skeletal Mesh=$System.Skeletal Mesh
   )
   &ParticleState(Kill Particles When Lifetime Has Elapsed=false, Let Infinitely Lived Particles Die When Emitter Deactivates=true)
   &ScaleSpriteSize(Scale Sprite Size Mode=::NewEnumerator0, Uniform Scale Factor=&Lerp_Float(A=1, B=1.25, Alpha=$System.PlumeLength))
   &ScaleColor(
    Scale Alpha=&Multiply_Float(A=$System.EquipFadeIn, B=&Lerp_Float(A=1, B=0.33, Alpha=$System.AlphaADS))
    Scale RGB=&MakeVectorFromLinearColorRGB(Color=&LerpLinearColors(StartColor=<rgba 25 7.5 0 1>, EndColor=<rgba 5 5 5 1>, LerpFactor=$User.Variant))
    ScaleA=true
    ScaleRGB=true
   )
   /SetVariables(
    Particles.ColorFakeLight:LinearColor=&MultiplyLinearColor(
     Linear Color=&LerpLinearColors(
      StartColor=<rgba 33 12.1 1.65 1>
      EndColor=&MakeCustomLinearColorFromBool(
       Bool=$User.bIsFrontEndPreview
       True Color=<rgba 30 30 30 0.3>
       False Color=&FNTimeOfDayColor(Night Color=<rgba 70 49.21 23.8 1>, Morning Color=<rgba 60 54 42 1>, Day Color=<rgba 22.9167 25 22.5 1>, Evening Color=<rgba 60 60 36 1>)
      )
      LerpFactor=$User.Variant
     )
     Scale Factor=&MakeVector4FromVectorAndFloat("Vector (RGB)"=<1 1 1>, "Float (Alpha)"=$System.EquipFadeIn)
    )
    Particles.PositionFakeLight:NiagaraPosition=&AddVectorToPosition(
     Position=$Particles.Position
     Vector=&RotateVectorByQuaternion(
      Vector To Rotate=&MergeHelper_VectorSwitch(0=<9 0 1>, 1=<-9 0 -1>, 2=?, 3=?, Index=$Particles.PartitionParticles.Partition, Num=::NewEnumerator0)
      Quaternion=$Particles.SkeletalMeshLocation.SampledRotation
     )
    )
    Particles.SpriteSizeFakeLight:Vector2f=&Multiply_Vector2D_ByFloat(Float=35, Vector2D=<1 1>)
   ) {isSetParameters=true, module=SetVariables_#8}
   /SetVariables(
    Particles.ColorGlow:LinearColor=&Multiply_LinearColorByFloat(
     Linear Color=&LerpLinearColors(
      StartColor=<rgba 1 0.4575 0.07 0.25>
      EndColor=&MakeCustomLinearColorFromBool(
       Bool=$User.bIsFrontEndPreview
       True Color=<rgba 1 1 1 0.1>
       False Color=&FNTimeOfDayColor(Night Color=<rgba 1 1 1 0.15>, Morning Color=<rgba 1 1 1 0.2>, Day Color=<rgba 1 1 1 0.2>, Evening Color=<rgba 1 1 1 0.2>)
      )
      LerpFactor=$User.Variant
     )
     Float=&Multiply_Float(A=$System.EquipFadeIn, B=&Lerp_Float(A=1, B=0.33, Alpha=$System.AlphaADS))
    )
    Particles.PositionGlow:NiagaraPosition=&AddVectorToPosition(
     Position=$Particles.Position
     Vector=&RotateVectorByQuaternion(
      Vector To Rotate=&MergeHelper_VectorSwitch(0=<12 0 3>, 1=<-12 0 -3>, 2=?, 3=?, Index=$Particles.PartitionParticles.Partition, Num=::NewEnumerator0)
      Quaternion=$Particles.SkeletalMeshLocation.SampledRotation
     )
    )
   ) {isSetParameters=true, module=SetVariables_#9}
   /SetVariables(
    Particles.CameraOffsetGlow=5
    Particles.SpriteSizeGlow:Vector2f=&Multiply_Vector2D_ByFloat(Float=3, Vector2D=$Particles.SpriteSize)
   ) {isSetParameters=true, module=SetVariables_#10}
   &FN_Light_Attributes(
    Color=&LerpLinearColors(StartColor=<rgba 36 16.47 2.52 1>, EndColor=<rgba 10 10 10 1>, LerpFactor=$User.Variant)
    Create Color Attribute=true
    Scale Color=&Waveform(
     Waveform Count=::NewEnumerator1
     "Waveform [1]"=::NewEnumerator0
     "[1] Input Position"=0.1
     "[1] Frequency"=0.433
     "[1] Amplitude Min/Max"=<0.4 1>
     "Waveform [2] Blend"=::NewEnumerator4
     "[2] Input Position"=0
     "[2] Phase"=$Emitter.Age
     "[2] Phase Scale"=1
     "[2] Compound Frequency"=2
     "[2] Frequency"=3.1
     "[2] Amplitude Min/Max"=<0.8 1>
     "[2] Scale"=1
     "[2] Bias"=0
     "[2] Exponent"=1
     "[2] Pulse Duration"=0.1
     "[2] Clip Triangle Min/Max"=<0 0>
     "[2] Random Seed"=0
    )
   ) {enabled=false}
   &DynamicMaterialParameters(
    Index 0 Param 2=10
    Index 0 Param 3=0.5
    Module.Index0Float0Write=false
    Module.Index0Float1Write=true
    Module.Index0Float2Write=true
    Module.Index0Float3Write=false
   )
   /SetVariables(
    Particles.ColorFlame:LinearColor=&MultiplyLinearColor(
     Linear Color=&LerpLinearColors(StartColor=<rgba 3 2 1.5 1>, EndColor=<rgba 1 1 1 1>, LerpFactor=$User.Variant)
     Scale Factor=&MakeVector4FromVectorAndFloat(
      "Vector (RGB)"=&VectorFromFloat(Value=&Lerp_Float(A=1, B=1.15, Alpha=$User.Variant))
      "Float (Alpha)"=&Multiply_Float(
       A=$System.EquipFadeIn
       B=&Multiply_Float(
        A=&Lerp_Float(A=1, B=&Lerp_Float(A=0.33, B=0.55, Alpha=$User.Variant), Alpha=$System.AlphaADS)
        B=&Lerp_Float(A=0.026, B=0.4, Alpha=$User.Variant)
       )
      )
     )
    )
    Particles.PositionFlame:NiagaraPosition=&AddVectorToPosition(
     Position=$Particles.Position
     Vector=&RotateVectorByQuaternion(
      Vector To Rotate=&MergeHelper_VectorSwitch(0=<0 0 0>, 1=<0 0 0>, 2=?, 3=?, Index=$Particles.PartitionParticles.Partition, Num=::NewEnumerator0)
      Quaternion=$Particles.SkeletalMeshLocation.SampledRotation
     )
    )
   ) {isSetParameters=true, module=SetVariables_#11}
   /SetVariables(Particles.OrientationUpperFlames:Quat4f=&MultiplyQuaternion(
     Quaternion A=$Particles.MeshOrientation
     Quaternion B=&MergeHelper_QuatSwitch(
      0:Quat4f=<-0.1 0.15 0 1>
      1:Quat4f=<-0.02 0.15 0 1>
      2:Quat4f=?
      3:Quat4f=?
      4:Quat4f=?
      5:Quat4f=?
      6:Quat4f=?
      7:Quat4f=?
      8:Quat4f=?
      9:Quat4f=?
      Index:NiagaraInt32=$Particles.PartitionParticles.Partition
      Num:MergeHelper_SwitcherEnum=::NewEnumerator0
     )
    )) {enabled=false, isSetParameters=true, module=SetVariables_#12}
  ]
  rendererClasses [&NiagaraSpriteRendererProperties, &NiagaraMeshRendererProperties]
  renderers [
   {rendererClass=&NiagaraSpriteRendererProperties}
   {rendererClass=&NiagaraSpriteRendererProperties, rendererIndex=1}
   {rendererClass=&NiagaraMeshRendererProperties, rendererIndex=2}
   {rendererClass=&NiagaraMeshRendererProperties, rendererIndex=3}
   {rendererClass=&NiagaraMeshRendererProperties, rendererIndex=4}
  ]
 }
}
summary {
 systemName=NS_Character_BoldDormRise_Idle_Body
 emitters [
  {emitterName=BasicSparks, rendererClasses [&NiagaraSpriteRendererProperties]}
  {emitterName=ShoulderFlames, rendererClasses [&NiagaraSpriteRendererProperties, &NiagaraMeshRendererProperties]}
  {emitterName=BeltFlames, rendererClasses [&NiagaraSpriteRendererProperties, &NiagaraMeshRendererProperties]}
  {
   bEnabled=false
   emitterName=BeltFlames001
   rendererClasses [&NiagaraSpriteRendererProperties, &NiagaraMeshRendererProperties]
  }
 ]
 userVariables [
  User.bIsADS {type=NiagaraBool}
  User.bIsFrontEnd {type=NiagaraBool}
  User.bIsFrontEndPreview {type=NiagaraBool}
  User.bisHeadHidden {type=NiagaraBool}
  User.Variant {type=NiagaraFloat, defaultValue=0, description=""}
 ]
}
